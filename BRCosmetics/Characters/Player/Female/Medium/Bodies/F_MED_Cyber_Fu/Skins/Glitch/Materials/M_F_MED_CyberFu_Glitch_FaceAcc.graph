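# Material for the face accessory of the female Cyber Fu character's Glitch skin, adding a procedural animated emissive pattern over the standard character material functions.
# Time-driven masks build a pulsing ring cut into stepped angular beams and broken by noise, with radial sine waves inside.
# Two offset, scaling and rotating grid-square samples crossfade to modulate the pattern.
# An emissive color and multiplier tint the result, which overrides Emissive before the cinematic post-FX function.
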Material /BRCosmetics/Characters/Player/Female/Medium/Bodies/F_MED_Cyber_Fu/Skins/Glitch/Materials/M_F_MED_CyberFu_Glitch_FaceAcc
graph {
 properties {
  bAutomaticallySetUsageInEditor=false
  bUseMaterialAttributes=true
  bUsedWithClothing=true
  bUsedWithMeshDeformer=true
  bUsedWithMorphTargets=true
  bUsedWithNanite=true
  bUsedWithSkeletalMesh=true
  pixelDepthOffsetMode=PDOM_Legacy
  refractionMethod=RM_IndexOfRefraction
  shadingModel=MSM_FromMaterialExpression
  subsurfaceProfile=&SS_HeroSkin_02
  usedShadingModels=Default Lit | Subsurface | Clear Coat | Subsurface Profile | Cloth | Eye
 }
 nodes [
  n1 = Subtract() {constB=4}
  Add(A=Constant() {r=2}, B=Multiply(A=Saturate(None=@n1)) {constB=4})
  ComponentMask(None=TextureCoordinate()) {r=true}
  ComponentMask(None=TextureCoordinate()) {r=true}
  Constant()
  Constant() {desc=0.795}
  n2 = MaterialFunctionCall() {materialFunction=&BaseFN_CharacterFunction_PreFX}
  n3 = MaterialFunctionCall(
   AO=@n2.AO
   Distance Blend=@n2.AO
   Fresnel=@n2.AO
   Gmap=@n2.AO
   Pre Skinned Local Position=@n2.AO
   Material=MaterialFunctionCall(
    PreskinnedPosition=@n2.Pre skinned Local Position
    Gmap=@n2.Pre skinned Local Position
    MaterialAttributes=@n2.Pre skinned Local Position
    AO=@n2.Pre skinned Local Position
   ) {materialFunction=&CharacterSkinFX}.Result
  ) {materialFunction=&BaseFN_CharacterFunction_PostFX}
  n4 = NamedRerouteUsage()
  n5 = Constant() {desc=RingRadius Outter, r=0.5}
  n6 = Constant() {desc=RingRadius Inner, r=0.4}
  n7 = SphereMask(A=@n4, B=@n5, Radius=Add(A=@n6) {constB=0.09}) {attenuationRadius=0.2, hardnessPercent=90}
  n8 = Frac(None=Multiply(A=Time()) {constB=0.4})
  n9 = Subtract(A=@n8) {constB=0.25}
  n10 = Saturate(None=Subtract(A=@n9) {constB=0.55})
  n11 = Multiply(A=Saturate(None=Subtract(A=@n9) {constB=0.795})) {constB=4}
  n12 = Add(A=Multiply(A=Subtract(A=@n10, B=@n11)) {constB=2})
  n13 = SphereMask(A=@n4, B=@n5, Radius=Multiply(A=@n6, B=@n12)) {attenuationRadius=0.2, hardnessPercent=90}
  n14 = Constant() {r=2}
  n15 = Floor(None=Multiply(A=@n8, B=Constant() {desc=Ring Steps, r=6}) {constB=10})
  n16 = Add(A=@n15)
  n17 = Min(A=Multiply(A=Power(Base=@n14, Exp=@n16), B=Saturate(None=@n15))) {constB=16}
  n18 = Divide(B=Divide(A=@n17) {constB=4})
  n19 = Constant() {r=-0.0625}
  n20 = Multiply(
   A=MaterialFunctionCall!(UVs=NamedRerouteUsage(), Rotation=Multiply(A=@n18, B=@n19)) {materialFunction=/FortUILibrary/MaterialFunctions/Gradients/MF_UI_AngleGradient}.Result
   B=@n17
  ) {constB=2}
  n21 = Power(Base=Saturate(None=Sine(None=@n20)), Exp=Multiply(B=Power(Base=@n18)) {constA=24}) {constExponent=8}
  n22 = Multiply(A=Constant() {r=0.6}, B=@n18) {constA=64}
  n23 = Multiply(
   A=Saturate(None=Subtract(A=@n7, B=@n13)) {desc=Ring}
   B=OneMinus(None=MaterialFunctionCall(In=@n21, Contrast=@n22) {materialFunction=&CheapContrast}.Result) {desc=Beams}
  ) {desc=Segments}
  n24 = Subtract(A=@n8) {constB=0.6}
  n25 = TextureSample(UVs=Multiply(A=NamedRerouteUsage()) {constB=2.7}) {samplerType=SAMPLERTYPE_Grayscale, texture=&LowResBlurredNoise}
  n26 = Add(A=Multiply(A=Saturate(None=@n24), B=Multiply(A=@n25.RGB) {constB=64})) {desc=Noise}
  n27 = Add(A=Multiply(A=@n8) {constB=1.36}) {constB=3.1625}
  n28 = Multiply(
   A=MaterialFunctionCall(UVs=NamedRerouteUsage(), Radius=@n27, Density=Constant() {r=0.5}) {materialFunction=&RadialGradientExponential}.RadialGradientExponential
   B=@n27
  ) {constB=2}
  n29 = MaterialFunctionCall(In=Abs(None=Sine(None=@n28) {desc=.4, period=0.6}), Contrast=Constant() {r=1.5}) {materialFunction=&CheapContrast}
  n30 = Sine(None=Add(A=@n28) {constB=0.5}) {desc=.4}
  n31 = OneMinus(None=Power(Base=Abs(None=@n30)) {constExponent=128})
  n32 = Multiply(A=Subtract(A=@n6) {constB=0.05}, B=@n12)
  n33 = NamedRerouteUsage()
  n34 = Multiply(A=Time()) {constB=0.4}
  n35 = Frac(None=@n34)
  n36 = MaterialFunctionCall(UVs=@n33, Texture Scale=Add(A=Multiply(A=@n35) {constB=2})) {materialFunction=&ScaleUVsByCenter}
  n37 = Constant() {r=0.25}
  n38 = Divide(A=@n34)
  n39 = Subtract(A=Multiply(A=Frac(None=@n38)) {constB=2})
  n40 = OneMinus(None=Ceil(None=Saturate(None=@n39)))
  n41 = TextureSample(UVs=MaterialFunctionCall(UVs=Add(A=@n36.UVs, B=@n37) {constB=0.2}, "Rotation Angle (0-1)"=Multiply(A=@n40) {constB=0.75}) {materialFunction=&CustomRotator}.Rotated Values) {samplerType=SAMPLERTYPE_LinearColor, texture=/Game/Packages/Fortress_SharedMaterials/Textures/T_Slowfield_GridSquares_01}
  n42 = Subtract(A=@n34) {constB=0.5}
  n43 = Frac(None=@n42)
  n44 = MaterialFunctionCall(UVs=@n33, Texture Scale=Add(A=Multiply(A=@n43) {constB=2}) {constB=2}) {materialFunction=&ScaleUVsByCenter}
  n45 = Multiply(A=Frac(None=Divide(A=@n42))) {constB=2}
  n46 = Ceil(None=Saturate(None=Subtract(A=@n45)))
  n47 = TextureSample(UVs=MaterialFunctionCall(UVs=Add(A=@n37, B=@n44.UVs) {constB=0.2}, "Rotation Angle (0-1)"=Multiply(A=@n46) {constB=0.25}) {materialFunction=&CustomRotator}.Rotated Values) {samplerType=SAMPLERTYPE_LinearColor, texture=&T_Slowfield_GridSquares_01}
  n48 = Add(A=@n34) {constB=0.1}
  n49 = MaterialFunctionCall(In=ConstantBiasScale(None=Sine(None=@n48)), Contrast=Constant() {r=0.25}) {materialFunction=&CheapContrast}
  n50 = OneMinus(None=MaterialFunctionCall(In=LinearInterpolate(A=@n41.R, B=@n47.R, Alpha=@n49.Result), Contrast=Constant() {r=1}) {materialFunction=&CheapContrast}.Result)
  n51 = Multiply(
   A=LinearInterpolate(
    A=Power(Base=@n23, Exp=@n26)
    B=Multiply(A=@n29.Result, B=@n31) {desc=Waves}
    Alpha=SphereMask(A=@n4, B=@n5, Radius=@n32) {attenuationRadius=0.2, hardnessPercent=90}
   )
   B=Add(A=Multiply(A=@n50) {constB=8}) {desc=Squares}
  )
  n52 = Multiply(
   A=VectorParameter() {group=Emissive, parameterName=Emissive Color, defaultValue=<rgba 1 1 1 1>}.RGB
   B=ScalarParameter() {defaultValue=1, group=Emissive, parameterName=emissive mult}
  )
  MaterialFunctionCall(In=SetMaterialAttributes(MaterialAttributes=@n3.Result, Emissive Color=Multiply(A=@n51, B=@n52)) {attributeSetTypes [#1]}) {materialFunction=&CINE_MF_CharPostFX}
  n53 = TextureCoordinate()
  n54 = Constant() {desc=UV Scale, r=0.25}
  NamedRerouteDeclaration(None=Add(
    A=MaterialFunctionCall(UVs=@n53, Texture Scale=@n54) {materialFunction=&ScaleUVsByCenter}.UVs
    B=Constant2Vector() {desc=UV Offset Topology Center, g=-0.5925, r=-0.01}
   )) {name=CenteredUVs, nodeColor=<rgba 0.2891 0 1 1>}
  n55 = Multiply(A=NamedRerouteUsage()) {constB=4}
  OneMinus(None=Power(Base=TextureSample(UVs=@n55) {samplerType=SAMPLERTYPE_Grayscale, texture=/Game/Global/Base4Layers_2022/Textures/EnDGrid}.R) {constExponent=0.1}) {desc=Grid}
  Saturate(None=Subtract() {constB=3})
 ]
}
